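# Base material for the BlankCanvas gadget emote prop, layering animated sky and tree images over the character parent material's base color.
# It sits between the parent PreFX and PostFX functions, blending in textures with adjustable UV offset and scale through masks.
# Tree_Animation drives a step reveal from a gradient texture.
# Sky_Animation drives a three-stage reveal, splitting stages by gradient blue and stepping against squared red.
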Material /BRCosmetics/Animation/Props/Gadgets_BlankCanvas/Materials/M_Gadget_BlankCanvas_Base
graph {
 properties {
  bUseMaterialAttributes=true
  bUsedWithInstancedStaticMeshes=true
  bUsedWithMeshDeformer=true
  bUsedWithMorphTargets=true
  bUsedWithSkeletalMesh=true
  pixelDepthOffsetMode=PDOM_Legacy
  shadingModel=MSM_FromMaterialExpression
  usedShadingModels=Default Lit | Subsurface | Clear Coat | Subsurface Profile | Cloth
 }
 nodes [
  n1 = MaterialFunctionCall() {materialFunction=&MF_CharacterParent_2023_PreFX}
  MaterialFunctionCall(Material=NamedRerouteUsage(), M=@n1.M, Pre Skinned Local Position=@n1.M, Distance Blend=@n1.M, Fresnel=@n1.M) {materialFunction=&MF_CharacterParent_2023_PostFX}
  NamedRerouteDeclaration(None=@n1.Result) {name=IN_Material, nodeColor=<rgba 0.5 0 1 1>}
  n2 = GetMaterialAttributes(None=NamedRerouteUsage()) {attributeGetTypes [#1]}
  n3 = LinearInterpolate(A=@n2.Base Color, B=NamedRerouteUsage(), Alpha=Multiply(A=NamedRerouteUsage(), B=NamedRerouteUsage()))
  n4 = NamedRerouteUsage()
  n5 = Multiply(A=NamedRerouteUsage(), B=NamedRerouteUsage())
  NamedRerouteDeclaration(None=SetMaterialAttributes!(MaterialAttributes=@n2.MaterialAttributes, Base Color=LinearInterpolate(A=@n3, B=@n4, Alpha=@n5)) {attributeSetTypes [#1]}) {name=OUT_Material, nodeColor=<rgba 0 1 0.1797 1>}
  n6 = TextureSample(UVs=NamedRerouteUsage()) {samplerType=SAMPLERTYPE_Masks, texture=/BRCosmetics/Animation/Props/Gadgets_BlankCanvas/Textures/T_BlankCanvas_TreeGradient}
  n7 = Multiply(A=Divide(A=Add(A=@n6.R, B=@n6.R))) {constB=10}
  n8 = ScalarParameter() {group=BlankCanvas, parameterName=Tree_Animation}
  NamedRerouteDeclaration(None=Step(Y=Subtract(A=@n7, B=@n6.G), X=Multiply(A=@n8) {constB=10}) {constX=0}) {name=Tree_Alpha, nodeColor=<rgba 1 0 0.3281 1>}
  n9 = MaterialFunctionCall(Float3=NamedRerouteUsage()) {materialFunction=&BreakOutFloat3Components}
  n10 = Constant()
  n11 = NamedRerouteDeclaration(None=If(A=@n9.B, "A > B"=Constant() {r=1}, "A == B"=@n10, "A < B"=@n10) {constB=0.666, equalsThreshold=0.1}) {name=Sky_FirstStep_Mask, nodeColor=<rgba 0 1 0.8125 1>}
  n12 = Constant()
  n13 = NamedRerouteDeclaration(None=If(A=@n9.B, "A > B"=@n12, "A == B"=@n12, "A < B"=Constant() {r=1}) {constB=0.666, equalsThreshold=0.1}) {name=Sky_ThirdStep_Mask, nodeColor=<rgba 1 0.4922 0 1>}
  NamedRerouteDeclaration(None=OneMinus(None=Max(A=@n11, B=@n13))) {name=Sky_SecondStep_Mask, nodeColor=<rgba 0.1719 0 1 1>}
  NamedRerouteDeclaration(None=TextureSample(UVs=NamedRerouteUsage()) {samplerType=SAMPLERTYPE_Masks, texture=/BRCosmetics/Animation/Props/Gadgets_BlankCanvas/Textures/T_BlankCanvas_SkyGradient}.RGB) {name=Sky_Gradient_Texture, nodeColor=<rgba 0 0.5312 1 1>}
  NamedRerouteDeclaration(None=@n1.Pre skinned Local Position) {name=PreSkinnedLocalPosition, nodeColor=<rgba 1 0 0.1406 1>}
  NamedRerouteDeclaration(None=@n1.Tech Art Mask) {name=TechArtMask, nodeColor=<rgba 0 0.4609 1 1>}
  n14 = VectorParameter() {
   group=BlankCanvas
   parameterName=Sky UVs Offset Scale
   channelNames {a=V Scale, b=U Scale, g=V Offset, r=U Offset}
   defaultValue=<rgba 0 0 1 1>
  }
  n15 = MaterialFunctionCall(UVs=Add(A=TextureCoordinate(), B=AppendVector(A=@n14.U Offset, B=@n14.U Offset)), Texture Scale=AppendVector(A=@n14.U Scale, B=@n14.U Scale)) {materialFunction=&ScaleUVsByCenter}
  n16 = NamedRerouteDeclaration(None=@n15.UVs) {name=Sky_UVs, nodeColor=<rgba 1 0.2109 0 1>}
  n17 = TextureSampleParameter2D(UVs=@n16) {
   group=BlankCanvas
   parameterName=Sky Texture
   samplerSource=SSM_Wrap_WorldGroupSettings
   texture=/BRCosmetics/Animation/Props/Gadgets_BlankCanvas/Textures/T_BlankCanvas_Emote_Prop_Sky_D
  }
  NamedRerouteDeclaration(None=@n17.RGB) {name=Sky_Texture, nodeColor=<rgba 0.1719 0 1 1>}
  n18 = TextureSampleParameter2D(UVs=@n16) {
   group=BlankCanvas
   parameterName=Sky Mask
   samplerSource=SSM_Wrap_WorldGroupSettings
   samplerType=SAMPLERTYPE_Masks
   texture=/BRCosmetics/Animation/Props/Gadgets_BlankCanvas/Textures/T_BlankCanvas_Emote_Prop_Sky_Mask
  }
  NamedRerouteDeclaration(None=Multiply(A=@n18.R, B=@n15.0-1 mask)) {name=Sky_Mask, nodeColor=<rgba 0.1484 1 0 1>}
  n19 = TextureCoordinate()
  n20 = VectorParameter() {
   group=BlankCanvas
   parameterName=Tree UVs Offset Scale
   channelNames {a=V Scale, b=U Scale, g=V Offset, r=U Offset}
   defaultValue=<rgba 0 0 1 1>
  }
  n21 = AppendVector(A=@n20.U Offset, B=@n20.U Offset)
  n22 = NamedRerouteDeclaration(None=MaterialFunctionCall(UVs=Add(A=@n19, B=@n21), Texture Scale=AppendVector(A=@n20.U Scale, B=@n20.U Scale)) {materialFunction=&ScaleUVsByCenter}.UVs) {name=Tree_UVs, nodeColor=<rgba 0.6406 1 0 1>}
  n23 = TextureSampleParameter2D(UVs=@n22) {
   group=BlankCanvas
   parameterName=Tree Texture
   samplerSource=SSM_Clamp_WorldGroupSettings
   texture=/BRCosmetics/Animation/Props/Gadgets_BlankCanvas/Textures/T_BlankCanvas_Emote_Prop_Tree_D
  }
  NamedRerouteDeclaration(None=@n23.RGB) {name=TreeTexture, nodeColor=<rgba 1 0 0.4687 1>}
  n24 = TextureSampleParameter2D(UVs=@n22) {
   group=BlankCanvas
   parameterName=Tree Mask
   samplerSource=SSM_Clamp_WorldGroupSettings
   samplerType=SAMPLERTYPE_Masks
   texture=/BRCosmetics/Animation/Props/Gadgets_BlankCanvas/Textures/T_BlankCanvas_Emote_Prop_Tree_Mask
  }
  NamedRerouteDeclaration(None=@n24.R) {name=Tree_Mask, nodeColor=<rgba 0 0.7891 1 1>}
  n25 = MaterialFunctionCall(Float3=NamedRerouteUsage()) {materialFunction=&BreakOutFloat3Components}
  n26 = Multiply(A=ScalarParameter() {group=BlankCanvas, parameterName=Sky_Animation, sliderMax=1}) {constB=3}
  n27 = Subtract(A=@n26)
  n28 = Max(A=Multiply(A=Saturate(None=@n26), B=NamedRerouteUsage()), B=Multiply(A=Saturate(None=@n27), B=NamedRerouteUsage()))
  n29 = Multiply(A=Saturate(None=Subtract(A=@n26) {constB=2}), B=NamedRerouteUsage())
  NamedRerouteDeclaration(None=Step(Y=Multiply(A=@n25.R, B=@n25.R), X=Max(A=@n28, B=@n29))) {name=Sky_Alpha, nodeColor=<rgba 0 1 0.0156 1>}
 ]
}
